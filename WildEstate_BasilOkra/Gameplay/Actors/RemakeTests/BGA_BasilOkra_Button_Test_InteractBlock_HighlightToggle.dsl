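; Test interactable button actor for BasilOkra that tries out interact block tracing with a highlight toggle, as named in its interaction prompt.
; It hides the button mesh on BeginPlay, shows it when interact focus is gained and hides it again when focus is lost.
; All visibility changes run only where cosmetic events can execute, and the overlap and tick events are disabled stubs.
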(event EventBeginPlay
  (|CallParentFunction)
  (Networking|CanExecuteCosmeticEvents
    (:True
      (Rendering|SetVisibility (Variables|Default|GetButtonMesh) false false))))

(event Collision|EventActorBeginOverlap
  (params (OtherActor :type Actor%20Object%20Reference))
  (|CallParentFunction OtherActor) ; This node is disabled and will not be called.\nDrag off pins to build functionality.
)

(event EventTick
  (params (DeltaSeconds :type Float%20%28single-precision%29))
  (|CallParentFunction DeltaSeconds) ; This node is disabled and will not be called.\nDrag off pins to build functionality.
)

(event InteractInterface|EventOnInteractFocused
  (params (FortPC :type Fort%20Player%20Controller%20Object%20Reference) (ContextInfo :type Fort%20Interact%20Context%20Info%20Object%20Reference))
  (Networking|CanExecuteCosmeticEvents
    (:True
      (Rendering|SetVisibility (Variables|Default|GetButtonMesh) true false))))

(event InteractInterface|EventOnInteractUnfocused
  (params (FortPC :type Fort%20Player%20Controller%20Object%20Reference) (NewContextInfo :type Fort%20Interact%20Context%20Info%20Object%20Reference))
  (Networking|CanExecuteCosmeticEvents
    (:True
      (Rendering|SetVisibility (Variables|Default|GetButtonMesh) false false))))

(fn BlueprintGetInteractionString
  (params (InteractingPawn :type Fort%20Pawn%20Object%20Reference) (InteractionBeingAttempted :type EInteractionBeingAttempted%20Enum :default "FirstInteraction"))
  (return "NSLOCTEXT(\"[E5FE835D5853709A94EAB66319593620]\", \"E4F80A5E4A19B39A6943E89E6BF92707\", \"Test - Interact Block Trace | Highlight Toggle\")"))

(fn ConstructionScript
  (|CallParentFunction))
