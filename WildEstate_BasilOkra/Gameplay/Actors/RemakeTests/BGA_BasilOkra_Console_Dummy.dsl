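; Dummy console actor in the BasilOkra remake tests that hosts five test button variants: interact block and interact overlap, each with a highlight-toggle version, plus visible mesh.
; On BeginPlay with authority only, SpawnButton runs a Sequence that spawns each button class at its own Billboard component's world transform.
; Each spawned button attaches to the Buttons component, keeping world transform and welding bodies.
; Tick, overlap and construction script are empty.
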(event EventBeginPlay
  (Networking|SwitchHasAuthority
    (:Authority
      (CallFunction|SpawnButton))))

(event Collision|EventActorBeginOverlap
  (params (OtherActor :type Actor%20Object%20Reference)))

(event EventTick
  (params (DeltaSeconds :type Float%20%28single-precision%29)))

(fn SpawnButton
  (Utilities|FlowControl|Sequence
    (:then_0
      (bind _returnvalue (Game|SpawnActorfromClass "/WildEstate_BasilOkra/Gameplay/Actors/RemakeTests/BGA_BasilOkra_Button_Test_InteractBlock.BGA_BasilOkra_Button_Test_InteractBlock_C" (Transformation|GetWorldTransform (Variables|Default|GetBillboard)) "AlwaysSpawn" "MultiplyWithRoot" 0 0 0))
      (Transformation|AttachActorToComponent _returnvalue (Variables|Default|GetButtons) "None" "KeepWorld" "KeepWorld" "KeepWorld" true))
    (:then_1
      (bind _returnvalue_2 (Game|SpawnActorfromClass "/WildEstate_BasilOkra/Gameplay/Actors/RemakeTests/BGA_BasilOkra_Button_Test_InteractBlock_HighlightToggle.BGA_BasilOkra_Button_Test_InteractBlock_HighlightToggle_C" (Transformation|GetWorldTransform (Variables|Default|GetBillboard1)) "AlwaysSpawn" "MultiplyWithRoot" 0 0 0))
      (Transformation|AttachActorToComponent _returnvalue_2 (Variables|Default|GetButtons) "None" "KeepWorld" "KeepWorld" "KeepWorld" true))
    (:then_2
      (bind _returnvalue_3 (Game|SpawnActorfromClass "/WildEstate_BasilOkra/Gameplay/Actors/RemakeTests/BGA_BasilOkra_Button_Test_InteractOverlap.BGA_BasilOkra_Button_Test_InteractOverlap_C" (Transformation|GetWorldTransform (Variables|Default|GetBillboard2)) "AlwaysSpawn" "MultiplyWithRoot" 0 0 0))
      (Transformation|AttachActorToComponent _returnvalue_3 (Variables|Default|GetButtons) "None" "KeepWorld" "KeepWorld" "KeepWorld" true))
    (:then_3
      (bind _returnvalue_4 (Game|SpawnActorfromClass "/WildEstate_BasilOkra/Gameplay/Actors/RemakeTests/BGA_BasilOkra_Button_Test_InteractOverlap_HighlightToggle.BGA_BasilOkra_Button_Test_InteractOverlap_HighlightToggle_C" (Transformation|GetWorldTransform (Variables|Default|GetBillboard3)) "AlwaysSpawn" "MultiplyWithRoot" 0 0 0))
      (Transformation|AttachActorToComponent _returnvalue_4 (Variables|Default|GetButtons) "None" "KeepWorld" "KeepWorld" "KeepWorld" true))
    (:then_4
      (bind _returnvalue_5 (Game|SpawnActorfromClass "/WildEstate_BasilOkra/Gameplay/Actors/RemakeTests/BGA_BasilOkra_Button_Test_VisibleMesh.BGA_BasilOkra_Button_Test_VisibleMesh_C" (Transformation|GetWorldTransform (Variables|Default|GetBillboard4)) "AlwaysSpawn" "MultiplyWithRoot" 0 0 0))
      (Transformation|AttachActorToComponent _returnvalue_5 (Variables|Default|GetButtons) "None" "KeepWorld" "KeepWorld" "KeepWorld" true))))

(fn ConstructionScript)
